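# Material function in the Units library category that converts a scalar value from centimeters to kilometers.
# The Cm input feeds a Divide node with a constant divisor of 100000, and the result leaves through the Km output.
# It is exposed to the material function library for reuse in material graphs.
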MaterialFunction /Engine/Functions/Engine_MaterialFunctions01/Units/Cm-to-Km
graph {
 properties {bExposeToLibrary=true, description=Converts between Centimeters and Kilometers, libraryCategoriesText [Units]}
 nodes [FunctionOutput(None=Divide(A=FunctionInput() {inputName=Cm, inputType=FunctionInput_Scalar}) {constB=100000}) {outputName=Km}]
}
